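# Physics asset for the Samurai Ultra Armor backpack cosmetic, simulating its hanging tassels and cord loops around a two-capsule backpack_root body.
# Left and right five-body tassel chains hang from the root with free swing and limited twist.
# Left and right five-body loop chains use tight swing limits and locked twist, with angularly free cross-constraints tying their ends into a closed loop.
# Two small left blade spheres attach to the root with free rotation.
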PhysicsAsset /BRCosmetics/Accessories/FORT_Backpacks/M_MED_Samurai_Ultra_Armor_Blades/Meshes/M_MED_Samurai_Ultra_Armor_Backpack_Physics
bodies {
 backpack_root [
  {shapeType=Capsule, length=15.8418, radius=6.885, center=<-4.3119 -9.1914 0.0258>, rotation=<rot 0.0072 -90 -89.0527>}
  {shapeType=Capsule, length=25.75, radius=15, center=<-14.7034 9.5134 -0.0004>, rotation=<rot 90 0 -2.9332>}
 ]
 bladeSmKinD_l [{shapeType=Sphere, radius=0.505}]
 bladeSmKinE_l [{shapeType=Sphere, radius=0.505}]
 loop01_l [{shapeType=Capsule, length=3.2236, radius=1.142, center=<3.3701 -0.058 0.0211>, rotation=<rot 0 -92.206 -89.1575>}]
 loop01_r [{shapeType=Capsule, length=2.8322, radius=1.1843, center=<3.0342 0.0747 -0.0306>, rotation=<rot 0 -86.848 -87.7962>}]
 loop02_l [{shapeType=Capsule, length=2.0616, radius=1.2655, center=<2.724 0 0>, rotation=<rot 90 0 0>}]
 loop02_r [{shapeType=Capsule, length=1.4291, radius=1.2643, center=<2.1395 0.0001 0>, rotation=<rot 90 0 0>}]
 loop03_l [{shapeType=Sphere, radius=1.625}]
 loop03_r [{shapeType=Sphere, radius=1.75, center=<0.783 0 0>}]
 loop04_l [{shapeType=Sphere, radius=2.125, center=<0.3703 0 0>}]
 loop04_r [{shapeType=Sphere, radius=2.125, center=<0.2695 0 0>}]
 loop05_l [{shapeType=Sphere, radius=1.875}]
 loop05_r [{shapeType=Sphere, radius=1.875}]
 tassleA_l [{shapeType=Capsule, length=2.1296, radius=1.125, center=<-2.0899 0 0>, rotation=<rot 90 0 0>}]
 tassleA_r [{shapeType=Capsule, length=1.7177, radius=1.2061, center=<2.1509 -0.0003 0>, rotation=<rot 90 0 0>}]
 tassleB_l [{shapeType=Capsule, length=3.6939, radius=1.0066, center=<-2.8703 0.0407 0>, rotation=<rot 90 0 0>}]
 tassleB_r [{shapeType=Capsule, length=3.1939, radius=1.2566, center=<2.918 0 0>, rotation=<rot 90 0 0>}]
 tassleC_l [{shapeType=Capsule, length=1.1807, radius=1.2592, center=<-1.8512 0.0001 0>, rotation=<rot 90 0 0>}]
 tassleC_r [{shapeType=Capsule, length=0.5, radius=1.25, center=<1.8141 0 0>, rotation=<rot 90 0 0>}]
 tassleD_l [{shapeType=Capsule, length=1, radius=1.9012, center=<-2.1793 0.0001 0>, rotation=<rot 90 0 0>}]
 tassleD_r [{shapeType=Capsule, length=0, radius=1.875, center=<1.8311 -0.2082 0.0422>, rotation=<rot 0 -87.6282 -89.3313>}]
 tassleE_l [{shapeType=Capsule, length=0, radius=2, center=<-2.9975 0.0001 0>, rotation=<rot 90 0 0>}]
 tassleE_r [{shapeType=Capsule, length=0, radius=2.375, center=<2.736 -0.0001 0>, rotation=<rot 90 0 0>}]
}
constraints [
 {
  bone1Name=tassleA_r
  bone2Name=backpack_root
  swing1LimitDegrees=30
  swing1Motion=Free
  swing2LimitDegrees=30
  swing2Motion=Free
  twistLimitDegrees=6
 }
 {
  bone1Name=tassleB_r
  bone2Name=tassleA_r
  swing1LimitDegrees=45
  swing1Motion=Free
  swing2LimitDegrees=45
  swing2Motion=Free
  twistLimitDegrees=6
 }
 {
  bone1Name=tassleC_r
  bone2Name=tassleB_r
  swing1LimitDegrees=45
  swing1Motion=Free
  swing2LimitDegrees=45
  swing2Motion=Free
  twistLimitDegrees=6
 }
 {
  bone1Name=tassleD_r
  bone2Name=tassleC_r
  swing1LimitDegrees=45
  swing1Motion=Free
  swing2LimitDegrees=45
  swing2Motion=Free
  twistLimitDegrees=6
 }
 {
  bone1Name=tassleE_r
  bone2Name=tassleD_r
  swing1LimitDegrees=45
  swing1Motion=Free
  swing2LimitDegrees=45
  swing2Motion=Free
  twistLimitDegrees=6
 }
 {
  bone1Name=tassleA_l
  bone2Name=backpack_root
  swing1LimitDegrees=30
  swing1Motion=Free
  swing2LimitDegrees=30
  swing2Motion=Free
  twistLimitDegrees=6
 }
 {
  bone1Name=tassleB_l
  bone2Name=tassleA_l
  swing1LimitDegrees=30
  swing1Motion=Free
  swing2LimitDegrees=30
  swing2Motion=Free
  twistLimitDegrees=6
 }
 {
  bone1Name=tassleC_l
  bone2Name=tassleB_l
  swing1LimitDegrees=30
  swing1Motion=Free
  swing2LimitDegrees=30
  swing2Motion=Free
  twistLimitDegrees=6
 }
 {
  bone1Name=tassleD_l
  bone2Name=tassleC_l
  swing1LimitDegrees=30
  swing1Motion=Free
  swing2LimitDegrees=30
  swing2Motion=Free
  twistLimitDegrees=6
 }
 {
  bone1Name=tassleE_l
  bone2Name=tassleD_l
  swing1LimitDegrees=30
  swing1Motion=Free
  swing2LimitDegrees=30
  swing2Motion=Free
  twistLimitDegrees=6
 }
 {bone1Name=loop01_l, bone2Name=backpack_root, swing1LimitDegrees=12, swing2LimitDegrees=12, twistLimitDegrees=6, twistMotion=Locked}
 {bone1Name=loop02_l, bone2Name=loop01_l, swing1LimitDegrees=12, swing2LimitDegrees=12, twistLimitDegrees=5, twistMotion=Locked}
 {bone1Name=loop03_l, bone2Name=loop02_l, swing1LimitDegrees=12, swing2LimitDegrees=12, twistLimitDegrees=5, twistMotion=Locked}
 {bone1Name=loop04_l, bone2Name=loop03_l, swing1LimitDegrees=12, swing2LimitDegrees=12, twistLimitDegrees=5, twistMotion=Locked}
 {bone1Name=loop01_r, bone2Name=backpack_root, swing1LimitDegrees=12, swing2LimitDegrees=12, twistLimitDegrees=45, twistMotion=Locked}
 {bone1Name=loop02_r, bone2Name=loop01_r, swing1LimitDegrees=12, swing2LimitDegrees=12, twistLimitDegrees=5, twistMotion=Locked}
 {bone1Name=loop03_r, bone2Name=loop02_r, swing1LimitDegrees=12, swing2LimitDegrees=12, twistLimitDegrees=5, twistMotion=Locked}
 {bone1Name=loop04_r, bone2Name=loop03_r, swing1LimitDegrees=12, swing2LimitDegrees=12, twistLimitDegrees=5, twistMotion=Locked}
 {bone1Name=loop05_r, bone2Name=loop04_r, swing1LimitDegrees=12, swing2LimitDegrees=12, twistLimitDegrees=10, twistMotion=Locked}
 {bone1Name=loop05_l, bone2Name=loop04_l, swing1LimitDegrees=12, swing2LimitDegrees=12, twistLimitDegrees=45, twistMotion=Locked}
 {
  bone1Name=loop04_l
  bone2Name=loop05_r
  swing1LimitDegrees=45
  swing1Motion=Free
  swing2LimitDegrees=45
  swing2Motion=Free
  twistLimitDegrees=45
  twistMotion=Free
 }
 {
  bone1Name=loop05_r
  bone2Name=loop05_l
  swing1LimitDegrees=45
  swing1Motion=Free
  swing2LimitDegrees=45
  swing2Motion=Free
  twistLimitDegrees=45
  twistMotion=Free
 }
 {
  bone1Name=bladeSmKinD_l
  bone2Name=backpack_root
  swing1LimitDegrees=45
  swing1Motion=Free
  swing2LimitDegrees=45
  swing2Motion=Free
  twistLimitDegrees=45
  twistMotion=Free
 }
 {
  bone1Name=bladeSmKinE_l
  bone2Name=backpack_root
  swing1LimitDegrees=45
  swing1Motion=Free
  swing2LimitDegrees=45
  swing2Motion=Free
  twistLimitDegrees=45
  twistMotion=Free
 }
]
