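# Anim Blueprint for the Samurai Ultra Armor blades backpack, posing and swaying its winged blades.
# A state machine picks base, dive, float or targeting poses, feeding RigidBody and AnimDynamics on each side's lead blade.
# Graduated rotation constraints fan that motion across the other blades, then spine bend yaws the backpack root.
# Outside skydiving, speed-weighted additive Sprint and Base poses layer onto the cached result.
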AnimBlueprint /BRCosmetics/Accessories/FORT_Backpacks/M_MED_Samurai_Ultra_Armor_Blades/Meshes/M_MED_Samurai_UltraArmor_Backpack_AnimBlueprint
target_skeleton=/Game/Accessories/FORT_Backpacks/M_MED_Samurai_Ultra_Armor_Blades/Meshes/M_MED_Samurai_Ultra_Armor_Backpack_Skeleton
graphs [
 AnimGraph {nodes [
   n1 = Get PawnSpeedRemap() -> (PawnSpeedRemap:double)
   n2 = M_MED_Samurai_Ultra_Armor_Backpack_ADD_PoseAsset : Sprint!(PoseWeight:float=@n1) -> (Pose:Pose Link Structure) {
    is=title AnimGraphNode_PoseByName
    poseAsset=&/M_MED_Samurai_Ultra_Armor_Backpack_ADD_PoseAsset
    poseName=Sprint
   }
   n3 = Apply Additive(Base=Use cached pose 'preSprintCache'(), Additive=@n2) {alphaBoolBlend {blendOption=Linear}}
   n4 = M_MED_Samurai_Ultra_Armor_Backpack_ADD_PoseAsset : Base!(PoseWeight:float=@n1) -> (Pose:Pose Link Structure) {
    is=title AnimGraphNode_PoseByName
    poseAsset=&M_MED_Samurai_Ultra_Armor_Backpack_ADD_PoseAsset
    poseName=Base
   }
   Output Pose(Result=Blend Poses by bool(
     BlendPose_0=Use cached pose 'preSprintCache'()
     BlendPose_1=Apply Additive(Base=@n3, Additive=@n4) {alphaBoolBlend {blendOption=Linear}}
     BlendTime_0=0.3
     BlendTime_1=0.3
     bActiveValue=Get bIsSkyDiving()
    ) {blendType=Linear}) {layerGroup=None}
   AnimGraphNode_StateMachine_0 = wingPosition() {is=title AnimGraphNode_StateMachine}
   n5 = RigidBody(ComponentPose=Local To Component(LocalPose=@AnimGraphNode_StateMachine_0)) {
    bOverrideWorldGravity=true
    overridePhysicsAsset=&/M_MED_Samurai_Ultra_Armor_Backpack_Physics
    simulationSpace=BaseBoneSpace
    alphaBoolBlend {blendOption=Linear}
    baseBoneRef {boneName=backpack_root}
    componentAppliedLinearAccClamp=<100 100 100>
    componentLinearAccScale=<0.5 0.5 0.5>
    componentLinearVelScale=<0.4 0.4 0.4>
    overrideWorldGravity {z=-980}
   }
   n6 = Anim Dynamics - Bone: bladeLgD_l(ComponentPose=@n5, Alpha=0.15, ExternalForce="0,0,0") {
    angularDampingOverride=1
    angularSpringConstant=100000
    bAngularSpring=true
    bEnableWind=true
    bGravityOverrideInSimSpace=true
    bOverrideAngularDamping=true
    bUseGravityOverride=true
    simulationSpace=RootRelative
    alphaBoolBlend {blendOption=Linear}
    boundBone {boneName=bladeLgD_l}
    componentAppliedLinearAccClamp=<1000 1000 1000>
    componentLinearAccScale=<0.55 0.55 0.55>
    componentLinearVelScale=<0.35 0.35 0.35>
    gravityOverride {z=-980}
    physicsBodyDefinitions [{
      sphereCollisionRadius=0
      boundBone {boneName=bladeLgD_l}
      boxExtents=<20 20 8>
      constraintSetup {angularLimitsMax=<3 3 5>, angularLimitsMin=<-3 -3 -10>, angularTarget=<1 0 0>}
      localJointOffset=<-10 -5 0>
     }]
    relativeSpaceBone {boneName=backpack_root}
   }
   n7 = Anim Dynamics - Bone: bladeLgD_r(ComponentPose=@n6, Alpha=0.15, ExternalForce="0,0,0") {
    angularDampingOverride=1
    angularSpringConstant=100000
    bAngularSpring=true
    bEnableWind=true
    bGravityOverrideInSimSpace=true
    bOverrideAngularDamping=true
    bUseGravityOverride=true
    simulationSpace=RootRelative
    alphaBoolBlend {blendOption=Linear}
    boundBone {boneName=bladeLgD_r}
    componentAppliedLinearAccClamp=<1000 1000 1000>
    componentLinearAccScale=<0.7 0.7 0.7>
    componentLinearVelScale=<0.3 0.3 0.3>
    gravityOverride {z=-980}
    physicsBodyDefinitions [{
      sphereCollisionRadius=0
      boundBone {boneName=bladeLgD_r}
      boxExtents=<30 15 5>
      constraintSetup {angularLimitsMax=<3 3 5>, angularLimitsMin=<-3 -3 -10>, angularTarget=<1 0 0>}
      localJointOffset=<-10 -5 0>
     }]
    relativeSpaceBone {boneName=backpack_root}
   }
   n8 = "Constraint - bladeSmE_l (R)"(ComponentPose=@n7, Alpha=0.95, ConstraintWeights_0=1) {
    alphaBoolBlend {blendOption=Linear}
    boneToModify {boneName=bladeSmE_l}
    constraintSetup [{transformType=Rotation, targetBone {boneName=bladeLgD_l}}]
    constraintWeights [1]
   }
   n9 = "Constraint - bladeSmD_l (R)"(ComponentPose=@n8, Alpha=0.8, ConstraintWeights_0=1) {
    alphaBoolBlend {blendOption=Linear}
    boneToModify {boneName=bladeSmD_l}
    constraintSetup [{transformType=Rotation, targetBone {boneName=bladeLgD_l}}]
    constraintWeights [1]
   }
   n10 = "Constraint - bladeSmC_l (R)"(ComponentPose=@n9, Alpha=0.65, ConstraintWeights_0=1) {
    alphaBoolBlend {blendOption=Linear}
    boneToModify {boneName=bladeSmC_l}
    constraintSetup [{transformType=Rotation, targetBone {boneName=bladeLgD_l}}]
    constraintWeights [1]
   }
   n11 = "Constraint - bladeSmB_l (R)"(ComponentPose=@n10, Alpha=0.55, ConstraintWeights_0=1) {
    alphaBoolBlend {blendOption=Linear}
    boneToModify {boneName=bladeSmB_l}
    constraintSetup [{transformType=Rotation, targetBone {boneName=bladeLgD_l}}]
    constraintWeights [1]
   }
   n12 = "Constraint - bladeLgC_l (R)"(ComponentPose=@n11, Alpha=0.9, ConstraintWeights_0=1) {
    alphaBoolBlend {blendOption=Linear}
    boneToModify {boneName=bladeLgC_l}
    constraintSetup [{transformType=Rotation, targetBone {boneName=bladeLgD_l}}]
    constraintWeights [1]
   }
   n13 = "Constraint - bladeLgB_l (R)"(ComponentPose=@n12, Alpha=0.8, ConstraintWeights_0=1) {
    alphaBoolBlend {blendOption=Linear}
    boneToModify {boneName=bladeLgB_l}
    constraintSetup [{transformType=Rotation, targetBone {boneName=bladeLgD_l}}]
    constraintWeights [1]
   }
   n14 = "Constraint - bladeLgA_l (R)"(ComponentPose=@n13, Alpha=0.5, ConstraintWeights_0=1) {
    alphaBoolBlend {blendOption=Linear}
    boneToModify {boneName=bladeLgA_l}
    constraintSetup [{transformType=Rotation, targetBone {boneName=bladeLgD_l}}]
    constraintWeights [1]
   }
   n15 = "Constraint - bladeSmA_l (R)"(ComponentPose=@n14, Alpha=0.3, ConstraintWeights_0=1) {
    alphaBoolBlend {blendOption=Linear}
    boneToModify {boneName=bladeSmA_l}
    constraintSetup [{transformType=Rotation, targetBone {boneName=bladeLgD_l}}]
    constraintWeights [1]
   }
   n16 = "Constraint - bladeSmE_r (R)"(ComponentPose=@n15, Alpha=0.95, ConstraintWeights_0=1) {
    alphaBoolBlend {blendOption=Linear}
    boneToModify {boneName=bladeSmE_r}
    constraintSetup [{transformType=Rotation, targetBone {boneName=bladeLgD_r}}]
    constraintWeights [1]
   }
   n17 = "Constraint - bladeSmD_r (R)"(ComponentPose=@n16, Alpha=0.8, ConstraintWeights_0=1) {
    alphaBoolBlend {blendOption=Linear}
    boneToModify {boneName=bladeSmD_r}
    constraintSetup [{transformType=Rotation, targetBone {boneName=bladeLgD_r}}]
    constraintWeights [1]
   }
   n18 = "Constraint - bladeSmC_r (R)"(ComponentPose=@n17, Alpha=0.65, ConstraintWeights_0=1) {
    alphaBoolBlend {blendOption=Linear}
    boneToModify {boneName=bladeSmC_r}
    constraintSetup [{transformType=Rotation, targetBone {boneName=bladeLgD_r}}]
    constraintWeights [1]
   }
   n19 = "Constraint - bladeSmB_r (R)"(ComponentPose=@n18, Alpha=0.55, ConstraintWeights_0=1) {
    alphaBoolBlend {blendOption=Linear}
    boneToModify {boneName=bladeSmB_r}
    constraintSetup [{transformType=Rotation, targetBone {boneName=bladeLgD_r}}]
    constraintWeights [1]
   }
   n20 = "Constraint - bladeLgC_r (R)"(ComponentPose=@n19, Alpha=0.9, ConstraintWeights_0=1) {
    alphaBoolBlend {blendOption=Linear}
    boneToModify {boneName=bladeLgC_r}
    constraintSetup [{transformType=Rotation, targetBone {boneName=bladeLgD_r}}]
    constraintWeights [1]
   }
   n21 = "Constraint - bladeLgB_r (R)"(ComponentPose=@n20, Alpha=0.8, ConstraintWeights_0=1) {
    alphaBoolBlend {blendOption=Linear}
    boneToModify {boneName=bladeLgB_r}
    constraintSetup [{transformType=Rotation, targetBone {boneName=bladeLgD_r}}]
    constraintWeights [1]
   }
   n22 = "Constraint - bladeLgA_r (R)"(ComponentPose=@n21, Alpha=0.5, ConstraintWeights_0=1) {
    alphaBoolBlend {blendOption=Linear}
    boneToModify {boneName=bladeLgA_r}
    constraintSetup [{transformType=Rotation, targetBone {boneName=bladeLgD_r}}]
    constraintWeights [1]
   }
   n23 = "Constraint - bladeSmA_r (R)"(ComponentPose=@n22, Alpha=0.3, ConstraintWeights_0=1) {
    alphaBoolBlend {blendOption=Linear}
    boneToModify {boneName=bladeSmA_r}
    constraintSetup [{transformType=Rotation, targetBone {boneName=bladeLgD_r}}]
    constraintWeights [1]
   }
   n24 = "Transform (Modify) Bone - Bone: backpack_root"(ComponentPose=@n23, Alpha=1, Translation="0,0,0", Rotation=Get backpack_rotate(), Scale="1,1,1") {rotationMode=BMM_Additive, rotationSpace=BCS_BoneSpace, alphaBoolBlend {blendOption=Linear}, boneToModify {boneName=backpack_root}}
   Save cached pose 'preSprintCache'(Pose=Component To Local(ComponentPose=@n24)) {cacheName=preSprintCache}
  ]}
 wingPosition {
  entry=Base
  path=AnimGraph.AnimGraphNode_StateMachine_0.wingPosition
  ids=id
  nodes [
   AnimStateNode_0 = Base() {is=title AnimStateNode}
   AnimStateNode_1 = Diving() {is=title AnimStateNode}
   AnimStateNode_2 = Floating() {is=title AnimStateNode}
   AnimStateNode_3 = Targeting() {is=title AnimStateNode}
   AnimStateTransitionNode_0 = AnimStateTransitionNode()
   AnimStateTransitionNode_1 = AnimStateTransitionNode()
   AnimStateTransitionNode_2 = AnimStateTransitionNode()
   AnimStateTransitionNode_3 = AnimStateTransitionNode()
   AnimStateTransitionNode_5 = AnimStateTransitionNode()
   AnimStateTransitionNode_6 = AnimStateTransitionNode()
   AnimStateTransitionNode_7 = AnimStateTransitionNode()
  ]
  links [
   {
    dst=Diving
    rule=AnimGraph.AnimGraphNode_StateMachine_0.wingPosition.AnimStateTransitionNode_0.Transition
    src=Base
    via=AnimStateTransitionNode_0
   }
   {
    dst=Targeting
    rule=AnimGraph.AnimGraphNode_StateMachine_0.wingPosition.AnimStateTransitionNode_6.Transition
    src=Base
    via=AnimStateTransitionNode_6
   }
   {
    dst=Base
    rule=AnimGraph.AnimGraphNode_StateMachine_0.wingPosition.AnimStateTransitionNode_3.Transition
    src=Diving
    via=AnimStateTransitionNode_3
   }
   {
    dst=Floating
    rule=AnimGraph.AnimGraphNode_StateMachine_0.wingPosition.AnimStateTransitionNode_1.Transition
    src=Diving
    via=AnimStateTransitionNode_1
   }
   {
    dst=Base
    rule=AnimGraph.AnimGraphNode_StateMachine_0.wingPosition.AnimStateTransitionNode_2.Transition
    src=Floating
    via=AnimStateTransitionNode_2
   }
   {
    dst=Diving
    rule=AnimGraph.AnimGraphNode_StateMachine_0.wingPosition.AnimStateTransitionNode_5.Transition
    src=Floating
    via=AnimStateTransitionNode_5
   }
   {
    dst=Base
    rule=AnimGraph.AnimGraphNode_StateMachine_0.wingPosition.AnimStateTransitionNode_7.Transition
    src=Targeting
    via=AnimStateTransitionNode_7
   }
  ]
 }
 Base {
  path=AnimGraph.AnimGraphNode_StateMachine_0.wingPosition.AnimStateNode_0.Base
  nodes [
   n1 = M_MED_Samurai_Ultra_Armor_Backpack_PoseAsset : Base!(PoseWeight:float=1) -> (Pose:Pose Link Structure) {
    is=title AnimGraphNode_PoseByName
    poseAsset=&/M_MED_Samurai_Ultra_Armor_Backpack_PoseAsset
    poseName=Base
   }
   Output Animation Pose(Result=@n1)
  ]
 }
 Diving {
  path=AnimGraph.AnimGraphNode_StateMachine_0.wingPosition.AnimStateNode_1.Diving
  nodes [
   n1 = M_MED_Samurai_Ultra_Armor_Backpack_PoseAsset : Sprint!(PoseWeight:float=1) -> (Pose:Pose Link Structure) {
    is=title AnimGraphNode_PoseByName
    poseAsset=&M_MED_Samurai_Ultra_Armor_Backpack_PoseAsset
    poseName=Sprint
   }
   Output Animation Pose(Result=@n1)
  ]
 }
 Floating {
  path=AnimGraph.AnimGraphNode_StateMachine_0.wingPosition.AnimStateNode_2.Floating
  nodes [
   n1 = M_MED_Samurai_Ultra_Armor_Backpack_PoseAsset : Float!(PoseWeight:float=1) -> (Pose:Pose Link Structure) {
    is=title AnimGraphNode_PoseByName
    poseAsset=&M_MED_Samurai_Ultra_Armor_Backpack_PoseAsset
    poseName=Float
   }
   Output Animation Pose(Result=@n1)
  ]
 }
 Targeting {
  path=AnimGraph.AnimGraphNode_StateMachine_0.wingPosition.AnimStateNode_3.Targeting
  nodes [
   n1 = M_MED_Samurai_Ultra_Armor_Backpack_PoseAsset : Sprint(PoseWeight=1) {
    poseAsset=&M_MED_Samurai_Ultra_Armor_Backpack_PoseAsset
    poseName=Sprint
   }
   Output Animation Pose(Result=@n1)
  ]
 }
 Transition {
  path=AnimGraph.AnimGraphNode_StateMachine_0.wingPosition.AnimStateTransitionNode_0.Transition
  nodes [Result(bCanEnterTransition=Get bIsSkydiveDiveMode())]
 }
 Transition {
  path=AnimGraph.AnimGraphNode_StateMachine_0.wingPosition.AnimStateTransitionNode_1.Transition
  nodes [Result(bCanEnterTransition=Get bIsSkydiveFloating())]
 }
 Transition {
  path=AnimGraph.AnimGraphNode_StateMachine_0.wingPosition.AnimStateTransitionNode_2.Transition
  nodes [Result(bCanEnterTransition=Get bIsParachuteOpened())]
 }
 Transition {
  path=AnimGraph.AnimGraphNode_StateMachine_0.wingPosition.AnimStateTransitionNode_3.Transition
  nodes [Result(bCanEnterTransition=Get bIsParachuteOpened())]
 }
 Transition {
  path=AnimGraph.AnimGraphNode_StateMachine_0.wingPosition.AnimStateTransitionNode_5.Transition
  nodes [Result(bCanEnterTransition=Get bIsSkydiveDiveMode())]
 }
 Transition {
  path=AnimGraph.AnimGraphNode_StateMachine_0.wingPosition.AnimStateTransitionNode_6.Transition
  nodes [Result(bCanEnterTransition=Get bIsTargeting())]
 }
 Transition {
  path=AnimGraph.AnimGraphNode_StateMachine_0.wingPosition.AnimStateTransitionNode_7.Transition
  nodes [Result(bCanEnterTransition=NOT Boolean(A=Get bIsTargeting()))]
 }
 EventGraph {nodes [
   Event BlueprintUpdateAnimation()
   n1 = GetOwnerSkeletalMeshComponent()
   n2 = BreakRotator(InRot=BreakTransform(InTransform=GetDeltaTransformFromRefPose(self=@n1, BoneName=spine_05, BaseName=spine_01)).Rotation)
   > Set backpack_rotate(backpack_rotate=MakeRotator(Yaw=MapRangeClamped(Value=@n2.Yaw, InRangeA=-5, InRangeB=-30, OutRangeB=25)))
   > Set PawnSpeedRemap(PawnSpeedRemap:double=MapRangeClamped(Value=Get PawnSpeed2D(), InRangeA=0, InRangeB=549)) -> (Output_Get:double)
   TryGetPawnOwner()
  ]}
]
variables [
 backpack_rotate {type=Rotator}
 PawnSpeedRemap {type=number}
]
